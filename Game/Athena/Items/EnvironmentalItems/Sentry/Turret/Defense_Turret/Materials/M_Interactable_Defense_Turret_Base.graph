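# Masked, lit material for the base of the interactable defense turret, with diffuse, normal and packed mask textures.
# A Blades scalar drives a hard SphereMask along the third UV channel's V axis, limited by vertex red, to reveal or hide those parts, apparently as opacity mask.
# An emissive texture is tinted by color and scale parameters, then modulated by a Fresnel function with an adjustable low end.
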Material /Game/Athena/Items/EnvironmentalItems/Sentry/Turret/Defense_Turret/Materials/M_Interactable_Defense_Turret_Base
graph {
 properties {blendMode=BLEND_Masked, pixelDepthOffsetMode=PDOM_Legacy, usedShadingModels=Default Lit}
 nodes [
  n1 = ComponentMask(None=TextureCoordinate() {coordinateIndex=2}) {g=true}
  LinearInterpolate(
   B=SphereMask(A=OneMinus(None=@n1), B=Constant(), Radius=ScalarParameter() {defaultValue=1, parameterName=Blades}, Hardness=Constant() {r=1})
   Alpha=VertexColor().R
  ) {constA=1}
  n2 = VectorParameter() {parameterName=EmissiveColor, defaultValue=<rgba 0 0 0 1>}
  n3 = ScalarParameter() {defaultValue=1, parameterName=EmissiveScale}
  Multiply(
   A=Multiply(A=TextureSample() {texture=/Game/Creative/Sets/PropSets/Playgrounds/Textures/T_Turret_Base_E}.RGB, B=Multiply(A=@n2.RGB, B=@n3))
   B=LinearInterpolate(
    A=ScalarParameter() {defaultValue=0.5, parameterName=FresnelEmissiveLow}
    Alpha=MaterialFunctionCall() {materialFunction=&MF_Fresnel_FX}.Fresnel
   )
  )
  TextureSample() {texture=/Game/Athena/Items/EnvironmentalItems/Sentry/Turret/Defense_Turret/Textures/T_Interactable_Defense_Turret_Base_D}
  TextureSample() {
   samplerType=SAMPLERTYPE_Normal
   texture=/Game/Athena/Items/EnvironmentalItems/Sentry/Turret/Defense_Turret/Textures/T_Interactable_Defense_Turret_Base_N
  }
  TextureSample() {
   samplerType=SAMPLERTYPE_Masks
   texture=/Game/Athena/Items/EnvironmentalItems/Sentry/Turret/Defense_Turret/Textures/T_Interactable_Defense_Turret_Base_S
  }
 ]
}
